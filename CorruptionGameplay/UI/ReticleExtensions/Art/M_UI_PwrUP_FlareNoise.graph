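# Translucent UI material that draws an animated, noisy radial flare, apparently for a power-up reticle extension.
# Seam-corrected radial UVs drive two panning samples of blurred noise, one distorted by a second pair of panning noise samples in regular UV space.
# The combined, intensified noise is masked by an exponential radial gradient and added to an adjustable center fill.
# The result blends between falloff and main colors and, scaled by FinalAlpha, forms the alpha.
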Material /CorruptionGameplay/UI/ReticleExtensions/Art/M_UI_PwrUP_FlareNoise
graph {
 properties {blendMode=BLEND_Translucent, materialDomain=MD_UI, pixelDepthOffsetMode=PDOM_Legacy, usedShadingModels=Default Lit}
 nodes [
  n1 = MaterialFunctionCall!(UVs) {materialFunction=/Game/Effects/Fort_Effects/Materials/_Functions/VectortoRadialCorrection}
  ComponentMask(None=@n1.Vector to Radial) {r=true}
  n2 = ScalarParameter() {defaultValue=2, parameterName=FillDensity}
  n3 = MaterialFunctionCall(Radius=ScalarParameter() {defaultValue=0.25, parameterName=FillRadius}, Density=@n2) {materialFunction=&RadialGradientExponential}
  n4 = ScalarParameter() {defaultValue=1, parameterName=FillAmount}
  n5 = ComponentMask(None=Saturate(None=@n1.Vector to Radial Reversed)) {r=true}
  n6 = Saturate(None=@n1.Vector to Reversed Mask)
  n7 = Saturate(None=@n5)
  n8 = LinearInterpolate(A=OneMinus(None=@n5), B=@n5, Alpha=Saturate(None=DotProduct(A=@n6, B=@n7)))
  n9 = MaterialFunctionCall(UVs=@n8, Texture Scale=ScalarParameter() {defaultValue=-0.75, parameterName=Scale}) {materialFunction=&ScaleUVsByCenter}
  n10 = Time()
  n11 = AppendVector(
   A=ScalarParameter() {defaultValue=0.125, parameterName=Speed_X_Texture_2}
   B=ScalarParameter() {defaultValue=-0.075, parameterName=Speed_Y_Texture_2}
  )
  n12 = MaterialFunctionCall(UVs=TextureCoordinate(), Texture Scale=ScalarParameter() {defaultValue=-0.75, parameterName=Scale}) {materialFunction=&ScaleUVsByCenter}
  n13 = AppendVector(
   A=ScalarParameter() {defaultValue=0.125, parameterName=Speed_X_Texture_2}
   B=ScalarParameter() {defaultValue=-0.075, parameterName=Speed_Y_Texture_2}
  )
  n14 = TextureCoordinate()
  n15 = Time()
  n16 = AppendVector(A=ScalarParameter() {defaultValue=0.25, parameterName=Speed_X}, B=ScalarParameter() {defaultValue=0.125, parameterName=Speed_Y})
  n17 = DotProduct(
   A=TextureSample(UVs=Panner(Coordinate=@n12.UVs, Speed=@n13)) {samplerType=SAMPLERTYPE_Grayscale, texture=&LowResBlurredNoise}.R
   B=TextureSample(UVs=Panner(Coordinate=@n14, Time=@n15, Speed=@n16)) {samplerType=SAMPLERTYPE_Grayscale, texture=&LowResBlurredNoise}.R
  )
  n18 = ScalarParameter() {defaultValue=0.25, parameterName=Distortion}
  n19 = TextureSample(UVs=Add(A=Panner(Coordinate=@n9.UVs, Time=@n10, Speed=@n11), B=Multiply(A=@n17, B=@n18))) {samplerType=SAMPLERTYPE_Grayscale, texture=&LowResBlurredNoise}
  n20 = ScalarParameter() {defaultValue=0.25, parameterName=Speed_X}
  n21 = ScalarParameter() {defaultValue=0.125, parameterName=Speed_Y}
  n22 = TextureSample(UVs=Panner(Coordinate=@n8, Time=@n10, Speed=AppendVector(A=@n20, B=@n21))) {samplerType=SAMPLERTYPE_Grayscale, texture=&LowResBlurredNoise}
  n23 = Saturate(None=DotProduct(A=Power(Base=@n19.R), B=Power(Base=@n22.R)))
  n24 = ScalarParameter() {defaultValue=6, parameterName=Intensity}
  n25 = DotProduct(
   A=MaterialFunctionCall(Density=@n2) {materialFunction=&RadialGradientExponential}.RadialGradientExponential
   B=Saturate(None=Multiply(A=@n23, B=@n24))
  )
  n26 = Saturate(None=Add(A=Multiply(A=@n3.RadialGradientExponential, B=@n4), B=Saturate(None=@n25)))
  LinearInterpolate(
   A=VectorParameter() {parameterName=Falloff_Color, defaultValue=<rgba 0 1 0 1>}.RGB
   B=VectorParameter() {parameterName=Color, defaultValue=<rgba 1 0 0 1>}.RGB
   Alpha=@n26
  )
  Multiply(A=@n26, B=ScalarParameter() {defaultValue=1, parameterName=FinalAlpha})
 ]
}
