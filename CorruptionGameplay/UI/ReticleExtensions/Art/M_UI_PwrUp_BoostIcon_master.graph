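# Translucent UI master material for a power-up boost icon in the reticle extensions, with a parameterized icon texture, colors and motion.
# Icon UVs pan along V, then pass through rotation and center-based scaling before sampling.
# The red channel, thresholded and sharpened, blends between Color and Color_Falloff.
# That edge mask also blends glow into solid alpha, which is scaled and masked by an exponential radial gradient to fade the borders.
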Material /CorruptionGameplay/UI/ReticleExtensions/Art/M_UI_PwrUp_BoostIcon_master
graph {
 properties {blendMode=BLEND_Translucent, materialDomain=MD_UI, pixelDepthOffsetMode=PDOM_Legacy, usedShadingModels=Default Lit}
 nodes [
  n1 = Panner(Coordinate=TextureCoordinate(), Speed=AppendVector(A=Constant(), B=ScalarParameter() {defaultValue=-2, parameterName=Pan}))
  n2 = ScalarParameter() {parameterName=Rotation}
  n3 = TextureSampleParameter2D(UVs=MaterialFunctionCall(
    UVs=MaterialFunctionCall(UVs=@n1, "Rotation Angle (0-1)"=@n2) {materialFunction=&CustomRotator}.Rotated Values
    Texture Scale=ScalarParameter() {defaultValue=1, parameterName=Scale}
   ) {materialFunction=&ScaleUVsByCenter}.UVs) {parameterName=Icon, texture=/FortniteCoreUI/UI/Login/Art/T-Icon-BackArrowk-64}
  n4 = Subtract(A=@n3.R, B=ScalarParameter() {defaultValue=0.25, parameterName=Subtract})
  n5 = Saturate(None=Multiply(A=Saturate(None=@n4), B=ScalarParameter() {defaultValue=8, parameterName=Edge}))
  LinearInterpolate(
   A=VectorParameter() {parameterName=Color, defaultValue=<rgba 0 1 0 1>}.RGB
   B=VectorParameter() {parameterName=Color_Falloff, defaultValue=<rgba 1 0 0 1>}.RGB
   Alpha=@n5
  )
  n6 = MaterialFunctionCall(Density=ScalarParameter() {defaultValue=64, parameterName=EdgeHardness}) {materialFunction=&RadialGradientExponential}
  n7 = LinearInterpolate(A=Multiply(A=@n3.R, B=ScalarParameter() {defaultValue=1, parameterName=GlowAmount}), Alpha=@n5)
  n8 = Multiply(A=Saturate(None=Saturate(None=@n7)), B=ScalarParameter() {defaultValue=1, parameterName=Final_Alpha})
  Saturate(None=DotProduct(A=Saturate(None=@n6.RadialGradientExponential), B=Saturate(None=@n8)))
 ]
}
