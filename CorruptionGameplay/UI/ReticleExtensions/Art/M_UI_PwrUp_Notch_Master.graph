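# Translucent UI master material that draws a tintable notch shape, apparently for power-up reticle extensions.
# A radial gradient circle is masked by a wedge cut from two hard-edged half-planes, one rotated and offset by parameters.
# The shape's opacity is scaled by a blend between two overall alpha parameters for animation.
# A separate post-rotation of the UVs, apparently through the customized UV, turns the whole notch.
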Material /CorruptionGameplay/UI/ReticleExtensions/Art/M_UI_PwrUp_Notch_Master
graph {
 properties {blendMode=BLEND_Translucent, materialDomain=MD_UI, numCustomizedUVs=1, pixelDepthOffsetMode=PDOM_Legacy, usedShadingModels=Default Lit}
 nodes [
  n1 = Saturate(None=MaterialFunctionCall(
    UVs=TextureCoordinate()
    Radius=ScalarParameter() {defaultValue=0.5, group=Properties, parameterName=Radius}
    Density=ScalarParameter() {defaultValue=256, group=EdgeHardness, parameterName=Density_Outer}
   ) {materialFunction=&RadialGradientExponential}.RadialGradientExponential)
  n2 = TextureCoordinate()
  n3 = ScalarParameter() {defaultValue=64, parameterName=EdgeHardness}
  n4 = Multiply(A=ComponentMask(None=Add(A=@n2) {constB=-0.5}) {r=true}, B=@n3)
  n5 = Add(
   A=MaterialFunctionCall(UVs=@n2, "Rotation Angle (0-1)"=ScalarParameter() {defaultValue=0.05, parameterName=Rotation}) {materialFunction=&CustomRotator}.Rotated Values
   B=ScalarParameter() {defaultValue=-0.55, parameterName=Offset}
  )
  n6 = Saturate(None=Multiply(A=ComponentMask(None=@n5) {r=true}, B=@n3))
  n7 = Saturate(None=DotProduct(A=Saturate(None=@n4), B=OneMinus(None=@n6)))
  Multiply(
   A=Saturate(None=DotProduct(A=@n1, B=@n7))
   B=LinearInterpolate(
    A=ScalarParameter() {defaultValue=1, group=Global, parameterName=Overall_Alpha}
    B=ScalarParameter() {defaultValue=1, group=Global, parameterName=Overall_Alpha_Stop}
    Alpha=ScalarParameter() {parameterName=AnimateOverallAlpha}
   )
  )
  Reroute(None=MaterialFunctionCall(UVs=TextureCoordinate(), "Rotation Angle (0-1)"=ScalarParameter() {parameterName=PostRotation}) {materialFunction=&CustomRotator}.Rotated Values)
  VectorParameter() {parameterName=Color, defaultValue=<rgba 1 1 1 1>}
 ]
}
