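# Translucent UI master material for power-up overheat reticle extensions, drawing procedural rings, dots or arc gauges.
# Subtracted radial gradients form the ring, which static switches cut into angular sections, fill as progress bars, extend with rotated clamp edges or replace with dots.
# The shape blends two colors and drives opacity, with optional FillPercent-driven heat-up of color and alpha.
# Optional vertical fade and angular mask trim opacity, and rotation and flip controls reorient UVs.
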Material /CorruptionGameplay/UI/ReticleExtensions/Art/M_UI_PwrUp_Overheat_Master
graph {
 properties {blendMode=BLEND_Translucent, materialDomain=MD_UI, numCustomizedUVs=1, pixelDepthOffsetMode=PDOM_Legacy, usedShadingModels=Default Lit}
 nodes [
  n1 = Saturate(None=MaterialFunctionCall(
    UVs=TextureCoordinate()
    Radius=ScalarParameter() {defaultValue=0.25, group=Properties, parameterName=Radius}
    Density=ScalarParameter() {defaultValue=16, group=EdgeHardness, parameterName=Density_Outer}
   ) {materialFunction=&RadialGradientExponential}.RadialGradientExponential)
  n2 = ScalarParameter() {group=Properties, parameterName=Radius}
  n3 = Multiply(A=ScalarParameter() {defaultValue=8, group=Properties, parameterName=RingWidth}) {constB=0.01}
  n4 = Saturate(None=MaterialFunctionCall(
    UVs=TextureCoordinate()
    Radius=Subtract(A=@n2, B=@n3)
    Density=ScalarParameter() {defaultValue=12, group=EdgeHardness, parameterName=Density_Inner}
   ) {materialFunction=&RadialGradientExponential}.RadialGradientExponential)
  n5 = Saturate(None=DotProduct(A=@n1, B=OneMinus(None=@n4)))
  n6 = MaterialFunctionCall!(Rotation_Fraction=ScalarParameter() {defaultValue=0.25, group=Section, parameterName=Section_Orientation}, TexCoord=TextureCoordinate()) {materialFunction=/FortUILibrary/Foundation/Art_ToolKit/AngledBoxes/MF_UI_Rotator_Fractions}
  n7 = Constant() {r=0.5}
  n8 = MaterialFunctionCall(Vector or UVs=Multiply(A=Subtract(A=@n6.Result, B=@n7)) {constB=2}) {materialFunction=&VectorToRadialValue}
  n9 = Multiply(A=ScalarParameter() {defaultValue=25, group=Section, parameterName=Section_Percentage}) {constB=0.01}
  n10 = ScalarParameter() {defaultValue=100, parameterName=FillPercent, sliderMax=100}
  n11 = LinearInterpolate(
   B=@n9
   Alpha=LinearInterpolate(
    A=ScalarParameter() {parameterName=FillPercent_CorrectionBottom, sliderMax=100}
    B=ScalarParameter() {defaultValue=1, parameterName=FillPercent_CorrectionTop, sliderMax=100}
    Alpha=Multiply(A=@n10) {constB=0.01}
   )
  )
  n12 = Multiply(A=Subtract(A=Saturate(None=@n8.Vector Converted to Angle), B=StaticSwitchParameter(True=@n11, False=@n9) {parameterName=IsProgressBar}) {constB=0.5}) {constB=640}
  n13 = Saturate(None=OneMinus(None=Saturate(None=@n12)))
  n14 = StaticSwitchParameter(True=Saturate(None=DotProduct(A=@n5, B=@n13)), False=@n5) {parameterName=IsSection}
  n15 = TextureCoordinate()
  n16 = ScalarParameter() {defaultValue=-0.125, parameterName=Angle_1_Rotation}
  n17 = ComponentMask(None=Add(
    A=MaterialFunctionCall(UVs=@n15, "Rotation Angle (0-1)"=@n16) {materialFunction=&CustomRotator}.Rotated Values
    B=ScalarParameter() {defaultValue=-0.5, parameterName=Angle_1_Offset}
   )) {r=true}
  n18 = ScalarParameter() {defaultValue=32, parameterName=Angle_1_EdgeHardness}
  n19 = OneMinus(None=Saturate(None=Multiply(A=@n17, B=@n18)))
  n20 = TextureCoordinate()
  n21 = ScalarParameter() {defaultValue=0.125, group=Properties, parameterName=ClampRadius}
  n22 = ScalarParameter() {defaultValue=16, group=EdgeHardness, parameterName=Density_Outer}
  n23 = OneMinus(None=Saturate(None=MaterialFunctionCall(UVs=@n20, Radius=@n21, Density=@n22) {materialFunction=&RadialGradientExponential}.RadialGradientExponential))
  n24 = DotProduct(A=DotProduct(A=DotProduct(A=@n19, B=@n13), B=@n23), B=@n1)
  n25 = Add(
   A=MaterialFunctionCall(UVs=TextureCoordinate(), "Rotation Angle (0-1)"=ScalarParameter() {defaultValue=-0.325, parameterName=Angle_2_Rotation}) {materialFunction=&CustomRotator}.Rotated Values
   B=ScalarParameter() {defaultValue=-0.5, parameterName=Angle_2_Offset}
  )
  n26 = Saturate(None=Multiply(A=ComponentMask(None=@n25) {r=true}, B=ScalarParameter() {defaultValue=32, parameterName=Angle_2_EdgeHardness}))
  n27 = DotProduct(A=DotProduct(A=DotProduct(A=@n26, B=@n13), B=@n23), B=@n1)
  n28 = Saturate(None=Add(A=@n14, B=Add(A=@n24, B=@n27)))
  n29 = Saturate(None=StaticSwitchParameter(True=@n1, False=StaticSwitchParameter(True=@n28, False=@n14) {parameterName=IsClamp}) {parameterName=IsDot})
  n30 = LinearInterpolate(
   A=VectorParameter() {group=Color, parameterName=Color_2, defaultValue=<rgba 0 1 0 1>}.RGB
   B=VectorParameter() {group=Color, parameterName=Color, defaultValue=<rgba 1 0 0 1>}.RGB
   Alpha=@n29
  )
  n31 = VectorParameter() {group=Color, parameterName=HeatupColor_2, defaultValue=<rgba 0 1 0 1>}
  n32 = VectorParameter() {group=Color, parameterName=HeatupColor, defaultValue=<rgba 1 0 0 1>}
  n33 = ScalarParameter() {defaultValue=100, parameterName=FillPercent, sliderMax=100}
  LinearInterpolate(
   A=@n30
   B=LinearInterpolate(A=@n30, B=LinearInterpolate(A=@n31.RGB, B=@n32.RGB, Alpha=@n29), Alpha=Multiply(A=@n33) {constB=0.01})
   Alpha=ScalarParameter() {parameterName=ShouldHeatUp}
  )
  n34 = MaterialFunctionCall(Rotation_Fraction=ScalarParameter() {group=Global, parameterName=Final_Rotation}, TexCoord=TextureCoordinate()) {materialFunction=&MF_UI_Rotator_Fractions}
  n35 = MaterialFunctionCall(Float2=@n34.Result) {materialFunction=&BreakOutFloat2Components}
  n36 = OneMinus(None=@n35.R)
  Reroute(None=LinearInterpolate(A=@n34.Result, B=AppendVector(A=@n36, B=@n35.G), Alpha=ScalarParameter() {parameterName=FlipFillDirection, sliderMax=100}))
  n37 = ScalarParameter() {defaultValue=1, group=Global, parameterName=Overall_Alpha}
  n38 = ScalarParameter() {defaultValue=1, group=Global, parameterName=Overall_Alpha_Stop}
  n39 = ScalarParameter() {parameterName=AnimateOverallAlpha}
  n40 = ScalarParameter() {defaultValue=1, parameterName=HeatupAlphaStart}
  n41 = ScalarParameter() {defaultValue=1, parameterName=HeatupAlphaStop}
  n42 = Multiply(A=ScalarParameter() {defaultValue=100, parameterName=FillPercent, sliderMax=100}) {constB=0.01}
  n43 = Multiply(
   A=@n29
   B=LinearInterpolate(
    A=LinearInterpolate(A=@n37, B=@n38, Alpha=@n39)
    B=LinearInterpolate(A=@n40, B=@n41, Alpha=@n42)
    Alpha=ScalarParameter() {parameterName=ShouldHeatUp}
   )
  )
  n44 = ComponentMask(None=TextureCoordinate()) {g=true}
  n45 = LinearInterpolate(A=@n44, B=OneMinus(None=@n44), Alpha=ScalarParameter() {parameterName=FlipFade})
  n46 = LinearInterpolate(A=Saturate(None=@n43), B=Saturate(None=DotProduct(A=@n43, B=@n45)), Alpha=ScalarParameter() {parameterName=Fade})
  n47 = MaterialFunctionCall(Rotation_Fraction=ScalarParameter() {defaultValue=0.25, group=Section, parameterName=Section_Orientation}, TexCoord=TextureCoordinate()) {materialFunction=&MF_UI_Rotator_Fractions}
  n48 = Constant() {r=0.5}
  n49 = MaterialFunctionCall(Vector or UVs=Multiply(A=Subtract(A=@n47.Result, B=@n48)) {constB=2}) {materialFunction=&VectorToRadialValue}
  n50 = ScalarParameter() {defaultValue=25, group=Section, parameterName=Section_MaskPercent}
  n51 = Multiply(A=Subtract(A=Saturate(None=@n49.Vector Converted to Angle), B=Multiply(A=@n50) {constB=0.01}) {constB=0.5}) {constB=512}
  n52 = Saturate(None=OneMinus(None=Saturate(None=@n51)))
  n53 = LinearInterpolate(A=@n52, B=OneMinus(None=@n52), Alpha=ScalarParameter() {parameterName=FlipMask})
  StaticSwitchParameter(True=Saturate(None=DotProduct(A=@n46, B=@n53)), False=@n46) {parameterName=HasMask}
  n54 = ScalarParameter() {defaultValue=0.25, group=Section, parameterName=Section_Orientation}
  n55 = TextureCoordinate()
  n56 = Multiply(A=Subtract(
    A=MaterialFunctionCall(Rotation_Fraction=@n54, TexCoord=@n55) {materialFunction=&MF_UI_Rotator_Fractions}.Result
    B=Constant() {r=0.5}
   )) {constB=2}
  Subtract(A=Saturate(None=MaterialFunctionCall(Vector or UVs=@n56) {materialFunction=&VectorToRadialValue}.Vector Converted to Angle)) {constB=0.5}
 ]
}
